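# Post-copy Anim Blueprint for the female Graffiti 01 cosmetic body, copying the owner's mesh pose and adding secondary motion.
# When the playlist uses custom character parts, the head bone shrinks to 0.01 scale.
# Weighted constraints position three zipper bones, then chained AnimDynamics nodes simulate zippers with angular springs and front ties with thigh-driven spherical limits.
# Customization load sets front-end alpha and sim-space variables, which the AnimGraph never reads.
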AnimBlueprint /BRCosmetics/Characters/Player/Female/Medium/Bodies/F_MED_Graffiti_01/Meshes/F_MED_Graffiti_Body_01_AnimBlueprint
target_skeleton=/Game/Characters/Player/Female/Medium/Bodies/F_MED_Graffiti_01/Meshes/F_MED_Graffiti_Body_01_Skeleton
graphs [
 AnimGraph {nodes [
   n1 = "Transform (Modify) Bone - Bone: head"!(ComponentPose=Local To Component(LocalPose=Use cached pose 'PreScale Cache'()), Alpha=1) -> (Pose) {scaleMode=BMM_Replace, alphaBoolBlend {blendOption=Linear}, boneToModify {boneName=head}, scale=<0.01 0.01 0.01>}
   n2 = Component To Local(ComponentPose=@n1)
   n3 = Use cached pose 'PreScale Cache'()
   n4 = Get GetPlayerListUsesCustomCharacterParts() -> (GetPlayerListUsesCustomCharacterParts:bool)
   n5 = "Constraint - dyn_front_zipper (PPPPP)"!(
    ComponentPose=Local To Component(LocalPose=Blend Poses by bool(BlendPose_0=@n2, BlendPose_1=@n3, BlendTime_0=0, BlendTime_1=0, bActiveValue=@n4) {blendType=Linear})
    Alpha=1
    ConstraintWeights_0=0.0412
    ConstraintWeights_1:float=0.5049
    ConstraintWeights_2:float=0.403
    ConstraintWeights_3:float=0.0255
    ConstraintWeights_4:float=0.0255
   ) -> (Pose) {
    alphaBoolBlend {blendOption=Linear}
    boneToModify {boneName=dyn_front_zipper}
    constraintSetup [
     {targetBone {boneName=spine_03}}
     {targetBone {boneName=spine_04}}
     {targetBone {boneName=spine_05}}
     {targetBone {boneName=clavicle_l}}
     {targetBone {boneName=clavicle_r}}
    ]
    constraintWeights [1,1,1,1,1]
   }
   n6 = "Constraint - dyn_L_zipper (PP)"!(ComponentPose=@n5, Alpha=1, ConstraintWeights_0=0.75, ConstraintWeights_1:float=0.25) -> (Pose) {
    alphaBoolBlend {blendOption=Linear}
    boneToModify {boneName=dyn_L_zipper}
    constraintSetup [{targetBone {boneName=pelvis}}, {targetBone {boneName=thigh_twist_01_l}}]
    constraintWeights [1,1]
   }
   n7 = "Constraint - dyn_R_zipper (PP)"!(ComponentPose=@n6, Alpha=1, ConstraintWeights_0=0.75, ConstraintWeights_1:float=0.25) -> (Pose) {
    alphaBoolBlend {blendOption=Linear}
    boneToModify {boneName=dyn_R_zipper}
    constraintSetup [{targetBone {boneName=pelvis}}, {targetBone {boneName=thigh_twist_01_r}}]
    constraintWeights [1,1]
   }
   n8 = Anim Dynamics - Bone: dyn_front_zipper!(SimulationSpace:AnimPhysSimSpaceType Enum, ComponentPose=@n7, Alpha=1) -> (Pose) {
    angularDampingOverride=1
    angularSpringConstant=3000
    bAngularSpring=true
    bOverrideAngularDamping=true
    bUseGravityOverride=true
    bUsePlanarLimit=false
    gravityScale=0.35
    simulationSpace=RootRelative
    alphaBoolBlend {blendOption=Linear}
    boundBone {boneName=dyn_front_zipper}
    componentAppliedLinearAccClamp=<5000 5000 5000>
    componentLinearAccScale=<1 1 1>
    componentLinearVelScale=<1 1 1>
    gravityOverride {z=-980}
    physicsBodyDefinitions [{
      boundBone {boneName=dyn_front_zipper}
      boxExtents=<18 18 18>
      constraintSetup {angularLimitsMax=<0 10 5>, angularLimitsMin=<0 -10 -25>, angularTarget=<1 0 0>}
      localJointOffset=<4 0 0>
     }]
   }
   n9 = Anim Dynamics - Bone: dyn_L_zipper!(SimulationSpace:AnimPhysSimSpaceType Enum, ComponentPose=@n8, Alpha=1) -> (Pose) {
    angularDampingOverride=1
    angularSpringConstant=1000
    bAngularSpring=true
    bOverrideAngularDamping=true
    bUseGravityOverride=true
    bUsePlanarLimit=false
    gravityScale=0.35
    simulationSpace=RootRelative
    alphaBoolBlend {blendOption=Linear}
    boundBone {boneName=dyn_L_zipper}
    componentAppliedLinearAccClamp=<5000 5000 5000>
    componentLinearAccScale=<1 1 1>
    componentLinearVelScale=<1 1 1>
    gravityOverride {z=-980}
    physicsBodyDefinitions [{
      boundBone {boneName=dyn_L_zipper}
      boxExtents=<10 10 10>
      constraintSetup {angularLimitsMax=<0 20 0>, angularLimitsMin=<0 -20 -30>, angularTarget=<1 0 0>}
      localJointOffset=<4 0 0>
     }]
   }
   n10 = Anim Dynamics - Bone: dyn_R_zipper!(SimulationSpace:AnimPhysSimSpaceType Enum, ComponentPose=@n9, Alpha=1) -> (Pose) {
    angularDampingOverride=1
    angularSpringConstant=1000
    bAngularSpring=true
    bOverrideAngularDamping=true
    bUseGravityOverride=true
    bUsePlanarLimit=false
    gravityScale=0.35
    simulationSpace=RootRelative
    alphaBoolBlend {blendOption=Linear}
    boundBone {boneName=dyn_R_zipper}
    componentAppliedLinearAccClamp=<5000 5000 5000>
    componentLinearAccScale=<1 1 1>
    componentLinearVelScale=<1 1 1>
    gravityOverride {z=-980}
    physicsBodyDefinitions [{
      boundBone {boneName=dyn_R_zipper}
      boxExtents=<13 13 13>
      constraintSetup {angularLimitsMax=<0 20 0>, angularLimitsMin=<0 -20 -30>, angularTarget=<1 0 0>}
      localJointOffset=<4 0 0>
     }]
   }
   n11 = Anim Dynamics - Bone: dyn_L_frontTie_1!(SimulationSpace:AnimPhysSimSpaceType Enum, ComponentPose=@n10, Alpha=1) -> (Pose) {
    angularDampingOverride=1
    angularSpringConstant=2000
    bOverrideAngularDamping=true
    bUseGravityOverride=true
    bUsePlanarLimit=false
    bUseSphericalLimits=true
    gravityScale=0.35
    simulationSpace=RootRelative
    alphaBoolBlend {blendOption=Linear}
    boundBone {boneName=dyn_L_frontTie_1}
    componentAppliedLinearAccClamp=<5000 5000 5000>
    componentLinearAccScale=<1 1 1>
    componentLinearVelScale=<1 1 1>
    gravityOverride {z=-980}
    physicsBodyDefinitions [{
      boundBone {boneName=dyn_L_frontTie_1}
      boxExtents=<11 11 11>
      constraintSetup {angularLimitsMax=<5 15 15>, angularLimitsMin=<-5 -15 -5>, angularTarget=<1 0 0>}
      localJointOffset=<2 0 0>
     }]
    sphericalLimits [{limitRadius=8, drivingBone {boneName=thigh_l}, sphereLocalOffset=<-10 0 0>}]
   }
   n12 = Anim Dynamics - Bone: dyn_L_frontTie_2!(SimulationSpace:AnimPhysSimSpaceType Enum, ComponentPose=@n11, Alpha=1) -> (Pose) {
    angularDampingOverride=1
    angularSpringConstant=2000
    bOverrideAngularDamping=true
    bUseGravityOverride=true
    bUsePlanarLimit=false
    bUseSphericalLimits=true
    gravityScale=0.35
    simulationSpace=RootRelative
    alphaBoolBlend {blendOption=Linear}
    boundBone {boneName=dyn_L_frontTie_2}
    componentAppliedLinearAccClamp=<5000 5000 5000>
    componentLinearAccScale=<1 1 1>
    componentLinearVelScale=<1 1 1>
    gravityOverride {z=-980}
    physicsBodyDefinitions [{
      boundBone {boneName=dyn_L_frontTie_2}
      boxExtents=<11 11 11>
      constraintSetup {angularLimitsMax=<5 15 15>, angularLimitsMin=<-5 -15 -15>, angularTarget=<1 0 0>}
      localJointOffset=<2 0 0>
     }]
    sphericalLimits [{limitRadius=9.5, drivingBone {boneName=thigh_l}, sphereLocalOffset=<-10 0 0>}]
   }
   n13 = Anim Dynamics - Bone: dyn_L_frontTie_3!(SimulationSpace:AnimPhysSimSpaceType Enum, ComponentPose=@n12, Alpha=1) -> (Pose) {
    angularDampingOverride=1
    angularSpringConstant=2000
    bOverrideAngularDamping=true
    bUseGravityOverride=true
    bUsePlanarLimit=false
    bUseSphericalLimits=true
    gravityScale=0.35
    simulationSpace=RootRelative
    alphaBoolBlend {blendOption=Linear}
    boundBone {boneName=dyn_L_frontTie_3}
    componentAppliedLinearAccClamp=<5000 5000 5000>
    componentLinearAccScale=<1 1 1>
    componentLinearVelScale=<1 1 1>
    gravityOverride {z=-980}
    physicsBodyDefinitions [{
      boundBone {boneName=dyn_L_frontTie_3}
      boxExtents=<11 11 11>
      constraintSetup {angularLimitsMax=<5 15 15>, angularLimitsMin=<-5 -15 -15>, angularTarget=<1 0 0>}
      localJointOffset=<2 0 0>
     }]
    sphericalLimits [{limitRadius=10, drivingBone {boneName=thigh_l}, sphereLocalOffset=<-10 0 0>}]
   }
   n14 = Anim Dynamics - Bone: dyn_R_frontTie_1!(SimulationSpace:AnimPhysSimSpaceType Enum, ComponentPose=@n13, Alpha=1) -> (Pose) {
    angularDampingOverride=1
    angularSpringConstant=2000
    bOverrideAngularDamping=true
    bUseGravityOverride=true
    bUsePlanarLimit=false
    bUseSphericalLimits=true
    gravityScale=0.35
    simulationSpace=RootRelative
    alphaBoolBlend {blendOption=Linear}
    boundBone {boneName=dyn_R_frontTie_1}
    componentAppliedLinearAccClamp=<5000 5000 5000>
    componentLinearAccScale=<1 1 1>
    componentLinearVelScale=<1 1 1>
    gravityOverride {z=-980}
    physicsBodyDefinitions [{
      boundBone {boneName=dyn_R_frontTie_1}
      boxExtents=<16 16 16>
      constraintSetup {angularLimitsMax=<5 15 15>, angularLimitsMin=<-5 -15 -5>, angularTarget=<1 0 0>}
      localJointOffset=<2 0 0>
     }]
    sphericalLimits [{limitRadius=8, drivingBone {boneName=thigh_r}, sphereLocalOffset=<10 0 0>}]
   }
   n15 = Anim Dynamics - Bone: dyn_R_frontTie_2!(SimulationSpace:AnimPhysSimSpaceType Enum, ComponentPose=@n14, Alpha=1) -> (Pose) {
    angularDampingOverride=1
    angularSpringConstant=2000
    bOverrideAngularDamping=true
    bUseGravityOverride=true
    bUsePlanarLimit=false
    bUseSphericalLimits=true
    gravityScale=0.35
    simulationSpace=RootRelative
    alphaBoolBlend {blendOption=Linear}
    boundBone {boneName=dyn_R_frontTie_2}
    componentAppliedLinearAccClamp=<5000 5000 5000>
    componentLinearAccScale=<1 1 1>
    componentLinearVelScale=<1 1 1>
    gravityOverride {z=-980}
    physicsBodyDefinitions [{
      boundBone {boneName=dyn_R_frontTie_2}
      boxExtents=<16 16 16>
      constraintSetup {angularLimitsMax=<5 15 15>, angularLimitsMin=<-5 -15 -15>, angularTarget=<1 0 0>}
      localJointOffset=<2 0 0>
     }]
    sphericalLimits [{limitRadius=9.5, drivingBone {boneName=thigh_r}, sphereLocalOffset=<10 0 0>}]
   }
   Output Pose(Result=Component To Local(ComponentPose=@n15)) {layerGroup=None}
   Save cached pose 'PreScale Cache'(Pose=Copy Pose From Mesh(SourceMeshComponent=Get MeshToCopy() -> (MeshToCopy:Skeletal Mesh Component Object Reference))) {cacheName=PreScale Cache}
  ]}
 EventGraph {nodes [
   Event OnCharacterCustomizationLoaded()
   > Set frontendAlphaScale(frontendAlphaScale:double=SelectFloat(A=0.75, B=1, bPickA=Get bIsInFrontEnd())) -> (Output_Get:double)
   > n1 = Branch(Condition=Get bIsInFrontEnd())
   > Set simSpaceChange(simSpaceChange:AnimPhysSimSpaceType Enum=World) -> (Output_Get:AnimPhysSimSpaceType Enum)
   Set simSpaceChange(execute=@n1.else)
   Event BlueprintInitializeAnimation()
   > Set MeshToCopy(MeshToCopy:Skeletal Mesh Component Object Reference=GetOwnerSkeletalMeshComponent()) -> (Output_Get:Skeletal Mesh Component Object Reference)
   > Set GetPlayerListUsesCustomCharacterParts(GetPlayerListUsesCustomCharacterParts:bool=GetPlaylistUsesCustomCharacterParts!() -> (ReturnValue:bool)) -> (Output_Get:bool)
  ]}
]
variables [
 frontendAlphaScale {type=number}
 GetPlayerListUsesCustomCharacterParts {type=boolean}
 MeshToCopy {type=SkeletalMeshComponent}
 simSpaceChange {type=AnimPhysSimSpaceType}
]
